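# Unlit debug material for Niagara Fluids shallow water, usable on mesh particles, that turns a simulation texture into height and collision channels.
# A static Interpolate switch reads the Tex parameter either through a clamped, filtered sample or a custom HLSL Load of individual unfiltered texels.
# Absolute red times Height Scale becomes X and green times Collision Scale becomes Y of the resulting float3.
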Material /NiagaraFluids/Materials/ShallowWater/DebugWaterMaterial
graph {
 properties {bUsedWithNiagaraMeshParticles=true, pixelDepthOffsetMode=PDOM_Legacy, shadingModel=MSM_Unlit, usedShadingModels=Unlit}
 nodes [
  MaterialExpressionTextureCoordinate_0 = TextureCoordinate()
  MaterialExpressionTextureObjectParameter_0 = TextureObjectParameter() {parameterName=Tex}
  n1 = Custom!(Texture=@MaterialExpressionTextureObjectParameter_0, UnitPos=@MaterialExpressionTextureCoordinate_0) {
   code="int2 sz;\r\nTexture.GetDimensions(sz.x, sz.y);\r\nfloat2 V = UnitPos * sz - .5;\r\nint2 V_int = round(V);\r\nreturn Texture.Load(int3(V_int.x, V_int.y, 0));"
   showCode=true
   inputs [
    {inputName=Texture, input {expression=&self:MaterialExpressionTextureObjectParameter_0}}
    {inputName=UnitPos, input {expression=&self:MaterialExpressionTextureCoordinate_0}}
   ]
  }
  n2 = MaterialFunctionCall(Float3=StaticSwitchParameter(
    True=TextureSample(UVs=@MaterialExpressionTextureCoordinate_0, Tex=@MaterialExpressionTextureObjectParameter_0) {samplerSource=SSM_Clamp_WorldGroupSettings}.RGB
    False=@n1
   ) {parameterName=Interpolate}) {materialFunction=&BreakOutFloat3Components}
  MaterialFunctionCall(
   X=Multiply(A=Abs(None=@n2.R), B=ScalarParameter() {defaultValue=0.01, parameterName=Height Scale})
   Y=Multiply(A=@n2.G, B=ScalarParameter() {defaultValue=1, parameterName=Collision Scale})
  ) {materialFunction=&MakeFloat3}
 ]
}
