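# Material function in the Niagara Fluids shallow water set that converts a 0–1 Unit coordinate into a UV inside a tiled atlas of Grid2D attributes.
# A Custom HLSL node derives the tile column and row from Attribute Index, then scales and offsets the coordinate into that tile.
# Num Tiles X typically equals the attribute count, and Num Tiles Y defaults to one, giving a single row of tiles.
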MaterialFunction /NiagaraFluids/Materials/ShallowWater/NiagaraGrid2DUnitToUV
graph {nodes [
  MaterialExpressionFunctionInput_0 = FunctionInput() {inputName=Unit, inputType=FunctionInput_Vector2}
  MaterialExpressionFunctionInput_1 = FunctionInput() {inputName=Attribute Index, inputType=FunctionInput_Scalar, sortPriority=1}
  MaterialExpressionFunctionInput_2 = FunctionInput() {description=Typically Num Attributes, inputName=Num Tiles X, inputType=FunctionInput_Scalar, sortPriority=2}
  MaterialExpressionFunctionInput_3 = FunctionInput(Preview=Constant() {r=1}) {bUsePreviewValueAsDefault=true, description=Typically 1, inputName=Num Tiles Y, inputType=FunctionInput_Scalar, sortPriority=3}
  n1 = Custom!(
   In_Unit=@MaterialExpressionFunctionInput_0
   AttributeIndex=@MaterialExpressionFunctionInput_1
   NumTilesX=@MaterialExpressionFunctionInput_2
   NumTilesY=@MaterialExpressionFunctionInput_3
  ) {
   code="int TileIndexX = AttributeIndex % NumTilesX;\r\nint TileIndexY = AttributeIndex / NumTilesX;\r\n\t\t\t\t\r\nreturn float2(In_Unit.x / NumTilesX + 1.0*TileIndexX/NumTilesX, In_Unit.y / NumTilesY + 1.0*TileIndexY/NumTilesY);"
   outputType=CMOT_Float2
   inputs [
    {inputName=In_Unit, input {expression=&self:MaterialExpressionFunctionInput_0}}
    {inputName=AttributeIndex, input {expression=&self:MaterialExpressionFunctionInput_1}}
    {inputName=NumTilesX, input {expression=&self:MaterialExpressionFunctionInput_2}}
    {inputName=NumTilesY, input {expression=&self:MaterialExpressionFunctionInput_3}}
   ]
  }
  FunctionOutput(None=@n1) {outputName=UV}
 ]}
