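# Material function that builds water surface height and normal from a 1D profile buffer texture, for Niagara Fluids shallow water.
# A Custom HLSL loop projects velocity-advected world XY onto evenly spaced directions, sampling the tiled profile with time scrolling and per-direction noise offsets.
# Samples are weighted toward the flow direction as speed rises.
# Normals come from neighbor-sample differences rotated per direction; the scaled sums output as float4 height plus unnormalized normal.
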MaterialFunction /NiagaraFluids/Materials/ShallowWater/ApplyProfileBuffer
graph {nodes [
  MaterialExpressionComponentMask_5 = ComponentMask(None=FunctionInput() {inputName=World Position}) {g=true, r=true}
  MaterialExpressionScalarParameter_8 = ScalarParameter() {defaultValue=1000, parameterName=Tile Width}
  MaterialExpressionFunctionInput_2 = FunctionInput() {inputName=Time, inputType=FunctionInput_Scalar}
  MaterialExpressionTextureObjectParameter_1 = TextureObjectParameter() {parameterName=ProfileBufferTex}
  MaterialExpressionFunctionInput_0 = FunctionInput() {inputName=Velocity, inputType=FunctionInput_Vector2}
  MaterialExpressionScalarParameter_9 = ScalarParameter() {defaultValue=2048, parameterName=Num Cells}
  MaterialExpressionScalarParameter_10 = ScalarParameter() {defaultValue=128, parameterName=Num Direction Samples}
  MaterialExpressionScalarParameter_11 = ScalarParameter() {parameterName=Advection Mult}
  MaterialExpressionScalarParameter_12 = ScalarParameter() {parameterName=Const Velocity Mult, sliderMax=5, sliderMin=3}
  MaterialExpressionScalarParameter_13 = ScalarParameter() {defaultValue=100, parameterName=Time Loop}
  MaterialExpressionScalarParameter_0 = ScalarParameter() {defaultValue=1, parameterName=ProfileBufferVelocity}
  n1 = Custom!(
   Simulation=@MaterialExpressionComponentMask_5
   TileWidth=@MaterialExpressionScalarParameter_8
   t=@MaterialExpressionFunctionInput_2
   ProfileBufferTex=@MaterialExpressionTextureObjectParameter_1
   Velocity=@MaterialExpressionFunctionInput_0
   NumCells=@MaterialExpressionScalarParameter_9
   NumDirectionSamples=@MaterialExpressionScalarParameter_10
   AdvectionMult=@MaterialExpressionScalarParameter_11
   ConstVelocityMult=@MaterialExpressionScalarParameter_12
   TimeLoop=@MaterialExpressionScalarParameter_13
   ProfileBufferVelocity=@MaterialExpressionScalarParameter_0
  ) {
   code="struct Functions\r\n{\r\n\r\n  float2 NiagaraGrid2DUnitToUV(float2 In_Unit, int AttributeIndex, int NumTilesX, int NumTilesY)\r\n  {\r\nint TileIndexX = AttributeIndex % NumTilesX;\r\nint TileIndexY = AttributeIndex / NumTilesX;\r\n\t\t\t\t\r\nreturn float2(In_Unit.x / NumTilesX + 1.0*TileIndexX/NumTilesX, In_Unit.y / NumTilesY + 1.0*TileIndexY/NumTilesY);\r\n  }\r\n\r\n};\r\n\r\nFunctions f;\r\n\r\nfloat Height = 0;\r\nfloat3 Normal = float3(0,0,0);\r\n//t  = fmod(t, TimeLoop);\r\n\r\nfor (int i = 0; i < NumDirectionSamples; ++i)\r\n{\r\n    float theta = i * 2. * 3.1415 / NumDirectionSamples;\r\n\r\n    float2 kvector = normalize(float2(cos(theta), sin(theta)));\r\n    \r\n    float AngleDot = smoothstep(.5, 1, dot(normalize(Velocity), kvector));\r\n//clamp(dot(normalize(Velocity), kvector), 0, 1);\r\n    AngleDot = lerp(1, AngleDot, clamp(length(Velocity), 0, 1));\r\n\r\n    theta = acos(dot(kvector, float2(1,0)));\r\n\r\n  \r\n\r\n    float p = dot(Simulation - Velocity * t * AdvectionMult/ProfileBufferVelocity, kvector);\r\n    \r\n    float eps = InterleavedGradientNoise(kvector,1);\r\n\r\n//    float OffsetMag = 0;//length(Velocity);\r\n\r\n    float p_unit = frac((p -  t * ConstVelocityMult) / TileWidth + .5 + eps);\r\n\r\n    float2 tmpUV = f.NiagaraGrid2DUnitToUV(float2(p_unit, .5), 0, 1, 1);\r\n    float TmpH = Texture2DSample(ProfileBufferTex,     \r\n    ProfileBufferTexSampler, tmpUV).x;\r\n\r\n    Height += TmpH * AngleDot;\r\n\r\nfloat dx = TileWidth / NumCells;\r\n\r\nfloat3 WorldPos;\r\nWorldPos.x = (p_unit - .5) * TileWidth;;\r\nWorldPos.y = 0;\r\nWorldPos.z = TmpH;\r\n\r\nfloat3 WorldPosRight = WorldPos;\r\nWorldPosRight.x += dx;\r\n\r\nfloat p_unit_right = frac(p_unit + 1./NumCells);\r\ntmpUV = f.NiagaraGrid2DUnitToUV(float2(p_unit_right, .5), 0, 1, 1);\r\n\r\nfloat h_right = Texture2DSample(ProfileBufferTex,  ProfileBufferTexSampler, tmpUV).x;\r\n\r\nWorldPosRight.z = h_right;\r\n\r\nfloat3 N = normalize(cross(WorldPosRight - WorldPos,  float3(0,1,0)));\r\n\r\n\r\n//  float3 N = float3(0,0,0);\r\n//   tmpUV = f.NiagaraGrid2DUnitToUV(float2(p_unit, .5), 1, 4, 1);\r\n//    N.x = Texture2DSample(ProfileBufferTex, ProfileBufferTexSampler, tmpUV).x;\r\n\r\n//    tmpUV = f.NiagaraGrid2DUnitToUV(float2(p_unit, .5), 2, 4, 1);\r\n//    N.y = Texture2DSample(ProfileBufferTex, ProfileBufferTexSampler, tmpUV).x;\r\n\r\n//    tmpUV = f.NiagaraGrid2DUnitToUV(float2(p_unit, .5), 3, 4, 1);\r\n//    N.z = Texture2DSample(ProfileBufferTex, ProfileBufferTexSampler, tmpUV).x;\r\n\r\n  // rotate normal\r\n   Normal += float3(N.x * cos(theta) - N.y*sin(theta), N.x * sin(theta) + N.y * cos(theta), N.z) * AngleDot;\r\n\r\n}\r\n\r\nreturn float4(Height, Normal) * 2. * 3.1415 / NumDirectionSamples;///NumDirectionSamples;"
   outputType=CMOT_Float4
   inputs [
    {inputName=Simulation, input {expression=&self:MaterialExpressionComponentMask_5}}
    {inputName=TileWidth, input {expression=&self:MaterialExpressionScalarParameter_8}}
    {inputName=t, input {expression=&self:MaterialExpressionFunctionInput_2}}
    {
     inputName=ProfileBufferTex
     input {expression=&self:MaterialExpressionTextureObjectParameter_1}
    }
    {inputName=Velocity, input {expression=&self:MaterialExpressionFunctionInput_0}}
    {inputName=NumCells, input {expression=&self:MaterialExpressionScalarParameter_9}}
    {
     inputName=NumDirectionSamples
     input {expression=&self:MaterialExpressionScalarParameter_10}
    }
    {inputName=AdvectionMult, input {expression=&self:MaterialExpressionScalarParameter_11}}
    {
     inputName=ConstVelocityMult
     input {expression=&self:MaterialExpressionScalarParameter_12}
    }
    {inputName=TimeLoop, input {expression=&self:MaterialExpressionScalarParameter_13}}
    {
     inputName=ProfileBufferVelocity
     input {expression=&self:MaterialExpressionScalarParameter_0}
    }
   ]
  }
  FunctionOutput(None=@n1)
 ]}
