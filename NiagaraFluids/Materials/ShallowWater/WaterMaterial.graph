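# Single Layer Water material for Niagara Fluids shallow water, masked and usable on Niagara mesh particles.
# Tex supplies height and velocity, NormalTex normals, and FoamTex tinted foam.
# An optional profile buffer, sampled from world position and velocity at two half-offset looping times and crossfaded to hide resets, adds detail to height and normals.
# A smoothstep fade from Tex, scaled by opacity, thresholds into a mask that also gates normals.
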Material /NiagaraFluids/Materials/ShallowWater/WaterMaterial
graph {
 properties {
  allowTranslucentCustomDepthWrites=true
  bScreenSpaceReflections=true
  bUsedWithNiagaraMeshParticles=true
  blendMode=BLEND_Masked
  pixelDepthOffsetMode=PDOM_Legacy
  shadingModel=MSM_SingleLayerWater
  translucencyLightingMode=TLM_Surface
  usedShadingModels=SingleLayerWater
 }
 nodes [
  n1 = TextureCoordinate()
  n2 = TextureSample(UVs=@n1, Tex=TextureObjectParameter() {parameterName=FoamTex, texture=&Black})
  Multiply(
   A=MaterialFunctionCall(X=@n2.R, Y=@n2.R, Z=@n2.R) {materialFunction=&MakeFloat3}.Result
   B=VectorParameter() {parameterName=Foam Color, defaultValue=<rgba 1 1 1 0>}.RGB
  )
  n3 = TextureObjectParameter() {parameterName=Tex, texture=&Black}
  n4 = TextureSample(UVs=@n1, Tex=@n3)
  MaterialExpressionSubtract_0 = Subtract(A=@n4.R, B=@n4.R)
  MaterialExpressionScalarParameter_9 = ScalarParameter() {defaultValue=2, parameterName=WaterFade}
  n5 = Custom!(a=@MaterialExpressionSubtract_0, f=@MaterialExpressionScalarParameter_9) {
   code="smoothstep(0, f, a)"
   outputType=CMOT_Float1
   inputs [
    {inputName=a, input {expression=&self:MaterialExpressionSubtract_0}}
    {inputName=f, input {expression=&self:MaterialExpressionScalarParameter_9}}
   ]
  }
  MaterialExpressionMultiply_3 = Multiply(B=Multiply(A=@n5, B=ScalarParameter() {defaultValue=1, parameterName=Opacity}) {constA=1}) {constA=1}
  n6 = Custom!(a=@MaterialExpressionMultiply_3) {code="a > 1e-5", inputs [{inputName=a, input {expression=&self:MaterialExpressionMultiply_3}}]}
  n7 = WorldPosition() {worldPositionShaderOffset=WPT_ExcludeAllShaderOffsets}
  MaterialExpressionTime_0 = Time()
  MaterialExpressionScalarParameter_13 = ScalarParameter() {defaultValue=100, parameterName=Time Loop}
  n8 = Custom!(t=@MaterialExpressionTime_0, TimeLoop=@MaterialExpressionScalarParameter_13) {
   code="\r\nfloat t1 = fmod(t, TimeLoop);\r\nfloat t2 = fmod(t + .5 * TimeLoop, TimeLoop);\r\n\r\nfloat nt = t1 / TimeLoop;\r\n\r\nfloat Alpha = 2.*abs(nt - .5);\r\nreturn float3(t1, t2,  Alpha);"
   inputs [
    {inputName=t, input {expression=&self:MaterialExpressionTime_0}}
    {inputName=TimeLoop, input {expression=&self:MaterialExpressionScalarParameter_13}}
   ]
  }
  n9 = MaterialFunctionCall(Float3=@n8) {materialFunction=&BreakOutFloat3Components}
  n10 = MaterialFunctionCall(X=TextureSample(UVs=@n1, Tex=@n3).B, Y=TextureSample(UVs=@n1, Tex=@n3).A) {materialFunction=&MakeFloat2}
  n11 = LinearInterpolate(
   A=MaterialFunctionCall!(World Position=@n7.XYZ, Time=@n9.R, Velocity=@n10.Result) {materialFunction=&/ApplyProfileBuffer}.Result
   B=MaterialFunctionCall(World Position=@n7.XYZ, Time=@n9.G, Velocity=@n10.Result) {materialFunction=&ApplyProfileBuffer}.Result
   Alpha=@n9.B
  ) {constB=0}
  n12 = ComponentMask(None=@n11) {a=true, b=true, g=true}
  n13 = TextureSample(UVs=@n1, Tex=TextureObjectParameter() {parameterName=NormalTex, texture=&Black})
  Multiply(A=@n6, B=StaticSwitchParameter(True=Add(A=@n12, B=@n13.RGB), False=@n13.RGB) {parameterName=Use Profile Buffer})
  ScalarParameter() {defaultValue=0.05, parameterName=Metallic}
  ScalarParameter() {defaultValue=0.9, parameterName=Specular}
  ScalarParameter() {parameterName=Roughness}
  SingleLayerWaterMaterialOutput!(
   ScatteringCoefficients=VectorParameter() {parameterName=Scatterieng, defaultValue=<rgba 0.0532 0.2396 0.7292 0>}.RGB
   AbsorptionCoefficients=VectorParameter() {parameterName=Absorption, defaultValue=<rgba 1 1 1 0>}.RGB
   PhaseG
   ColorScaleBehindWater
  )
  n14 = Add(A=@n4.R, B=ComponentMask(None=@n11) {r=true})
  Transform(None=MaterialFunctionCall(Z=StaticSwitchParameter(True=@n14, False=@n4.R) {parameterName=Use Profile Buffer}) {materialFunction=&MakeFloat3}.Result) {transformSourceType=TRANSFORMSOURCE_Local}
  VectorParameter() {parameterName=Water Color, defaultValue=<rgba 1 1 1 0>}
 ]
}
