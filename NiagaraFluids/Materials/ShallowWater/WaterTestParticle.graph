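# Unlit, masked test material for Niagara sprite particles in the Niagara Fluids shallow water content.
# A Distance node measures each UV from the sprite centre, and a Custom HLSL node tests whether it is under half, producing a hard-edged circular disc value.
# The graph also holds a constant blue color and a Particle Color node.
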Material /NiagaraFluids/Materials/ShallowWater/WaterTestParticle
graph {
 properties {bUsedWithNiagaraSprites=true, blendMode=BLEND_Masked, pixelDepthOffsetMode=PDOM_Legacy, shadingModel=MSM_Unlit, usedShadingModels=Unlit}
 nodes [
  Constant3Vector() {constant=<rgba 0.1398 0.2809 0.8385 0>}
  MaterialExpressionDistance_0 = Distance(A=TextureCoordinate(), B=Constant2Vector() {g=0.5, r=0.5})
  Custom!(a=@MaterialExpressionDistance_0) {
   code="return  a < .5;"
   inputs [{inputName=a, input {expression=&self:MaterialExpressionDistance_0}}]
  }
  ParticleColor()
 ]
}
