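# Unlit, additive, two-sided material for whitewater particles in Niagara Fluids shallow water, usable on sprites and ribbons with responsive AA.
# A WhitewaterColor parameter supplies the tint, which a clamped TwoSidedSign lerp dims to a small fraction on back faces.
# Its alpha is multiplied by a UV-centered SphereMask with Radius and Hardness parameters, forming a round procedural falloff without textures.
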Material /NiagaraFluids/Materials/ShallowWater/WhitewaterMaterial
graph {
 properties {
  bEnableResponsiveAA=true
  bUsedWithNiagaraRibbons=true
  bUsedWithNiagaraSprites=true
  bUsedWithParticleSprites=true
  blendMode=BLEND_Additive
  pixelDepthOffsetMode=PDOM_Legacy
  shadingModel=MSM_Unlit
  twoSided=true
  usedShadingModels=Unlit
 }
 nodes [
  n1 = VectorParameter() {parameterName=WhitewaterColor, defaultValue=<rgba 0.9635 0.9776 1 0.3>}
  LinearInterpolate(A=Multiply(A=@n1.RGB) {constB=0.15}, B=@n1.RGB, Alpha=Clamp(None=TwoSidedSign()))
  Multiply(
   A=@n1.A
   B=SphereMask(
    A=TextureCoordinate()
    B=Constant() {r=0.5}
    Radius=ScalarParameter() {defaultValue=0.5, parameterName=Radius}
    Hardness=ScalarParameter() {defaultValue=0.1, parameterName=Hardness}
   )
  )
 ]
}
